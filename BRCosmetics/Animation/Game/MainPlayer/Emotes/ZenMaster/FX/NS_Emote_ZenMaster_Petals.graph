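# Niagara effect for the ZenMaster emote that sends a short streak of petals and magenta glints along local -X.
# Invisible GPU and CPU "Low" tracer particles follow a curve, each feeding two sprite emitters through particle-read spawning.
# Spawned sprites tumble via mesh-orientation alignment, drag and curl noise, with wind only on GPU emitters.
# A distance-culled Directional_Flare travels a similar path with a sprite and particle light.
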NiagaraSystem /BRCosmetics/Animation/Game/MainPlayer/Emotes/ZenMaster/FX/NS_Emote_ZenMaster_Petals
dependencies {
 usedDataInterfaces [
  &NiagaraDataInterfaceParticleRead
  &NiagaraDataInterfaceCurve
  &NiagaraDataInterfaceCollisionQuery
  &NiagaraDataInterfaceLandscape
  &NiagaraDataInterfaceVectorCurve
  &NiagaraDataInterfaceColorCurve
  &NiagaraDataInterfaceSocketReader
  &NiagaraDataInterfaceSkeletalMesh
 ]
 usedDynamicInputs []
 usedModules [
  &self:SystemSpawnScript.SystemScriptSource.SystemScriptGraph.NiagaraNodeAssignment_0.SetVariables_#1
  &SystemState
 ]
}
emitters {
 Directional_Flare {
  emitterUpdateScript [
   &EmitterState(
    Enable Distance Culling=true
    Life Cycle Mode=::NewEnumerator1
    Loop Behavior=::NewEnumerator1
    Loop Duration=1
    Max Distance Response=::KillImmediately
    MaxDistance=3500
    Module.bUseMaxDistance=true
    Module.bUseMinDistance=false
    Scalability Mode=::NewEnumerator1
    UseLoopDelay=true
   )
   &SpawnBurst_Instantaneous(
    Spawn Count=2
    Spawn Probability=&FN_FrontEndFloatModifier(Gameplay=2, Front End=0)
    Use Spawn Probability=false
   )
  ]
  particleSpawnScript [
   &InitializeParticle(
    Color=<rgba 7.5 0.75 6.4875 1>
    Lifetime=1.4
    Position Offset=<0 0 5>
    Sprite Size Max=<3.5 5>
    Sprite Size Min=<3.5 5>
    Sprite Size Mode=::NewEnumerator2
    UsePositionOffset=true
    Write Material Random=true
   ) {module=InitializeParticle001}
   &DynamicMaterialParameters(
    Index 0 Param 1=0.2
    Index 0 Param 2=1
    Index 0 Param 3=1
    Module.Param0WriteEnabled=true
    Module.Param1WriteEnabled=true
    Module.Param2WriteEnabled=true
    Module.Param3WriteEnabled=true
   )
  ]
  particleUpdateScript [
   &ParticleState()
   &ScaleSpriteSize(
    Uniform Curve Scale=1.5
    Uniform Curve Sprite Scale=DataInterface() {propertyValues {Curve {keys [
        {tangentMode=RCTM_User, time=0, value=1}
        <at 0.1422 1>
        <at 0.1786 1.4186>
        <at 0.3 2>
        <at 0.3511 1.5>
        <at 0.4722 1.5>
        <at 0.6 2>
        <at 0.7347 1.9688>
        <at 0.8 1.3>
        <at 0.8904 1.4>
        {tangentMode=RCTM_User, time=1, value=1.5}
       ]}}}
   )
   /SetVariables(Particles.Position=&VectorFromCurve(
     VectorCurve=DataInterface() {propertyValues {XCurve {keys [<at 0 0>,<at 1 130>]}, YCurve {keys [<at 0 0>,<at 1 0>]}, ZCurve {keys [<at 0 0>,<at 1 0>]}}}
     Scale Curve=<-1 1 1>
    )) {isSetParameters=true}
   &FN_Light_Attributes(
    Color=<rgba 1.2 0.1825 1.0811 1>
    Create Color Attribute=true
    Create Exponent Attribute=true
    Create Position Attribute=true
    Create Radius Attribute=true
    Enable Particle Light Enable=true
    Enable Visibility Tag=true
    Exponent=5
    Light Enabled=true
    Position=&VectorFromCurve(
     VectorCurve=DataInterface() {propertyValues {XCurve {keys [<at 0 0>, {arriveTangent=16.9116, time=1, value=105}]}, YCurve {keys [<at 0 0>,<at 0 0>]}, ZCurve {keys [<at 0 0>,<at 1 0>]}}}
     Scale Curve=<-1 1 1>
    )
    Radius=90
    Scale Color=&FloatFromCurve(FloatCurve=DataInterface() {propertyValues {Curve {keys [
         <at 0 0 RCIM_Cubic>
         {interpMode=RCIM_Cubic, tangentMode=RCTM_User, time=0.2, value=1}
         {arriveTangent=-1.25, interpMode=RCIM_Cubic, leaveTangent=-1.25, time=0.7591, value=0.85}
         {arriveTangent=-0.5426, interpMode=RCIM_Cubic, leaveTangent=-0.5426, tangentMode=RCTM_User, time=1, value=0}
        ]}}})
   ) {module=FN_Light_Attributes001}
   &ScaleColor(
    Scale Alpha=&FloatFromCurve(FloatCurve=DataInterface() {propertyValues {Curve {keys [<at 0 0>,<at 0.05 1>,<at 0.2 1>,<at 0.8 1>,<at 1 0>]}}})
    ScaleA=true
    ScaleRGB=true
   ) {module=ScaleColor001}
   &ScaleSpriteSize(Uniform Curve Sprite Scale=DataInterface() {propertyValues {Curve {keys [
        <at 0 0>
        {tangentMode=RCTM_User, time=0.1, value=1}
        {arriveTangent=-1.1494, interpMode=RCIM_Cubic, leaveTangent=-1.1494, time=0.65, value=1}
        <at 0.97 0 RCIM_Cubic>
       ]}}}) {module=ScaleSpriteSize002}
   &CameraOffset(Camera Offset Amount=5) {module=CameraOffset001}
  ]
  rendererClasses [&NiagaraSpriteRendererProperties, &NiagaraLightRendererProperties]
  renderers [
   {rendererClass=&NiagaraSpriteRendererProperties}
   {rendererClass=&NiagaraLightRendererProperties, rendererIndex=1}
  ]
 }
 Low_Petal_Plume {
  emitterUpdateScript [
   &EmitterState(Life Cycle Mode=::NewEnumerator1, Loop Behavior=::NewEnumerator1, Loop Duration=1.9, UseLoopDelay=false)
   &SpawnParticlesFromOtherEmitter(
    Attribute Reader=DataInterface() {propertyValues {EmitterBinding {bindingMode=Other, emitterName=Low_Petal_Tracer}}}
    Spawn Rate=50
    Use Maximum Age=false
    Use Minimum Age=false
   )
  ]
  particleSpawnScript [
   &InitializeParticle(
    Lifetime Max=1.6731
    Lifetime Min=0.666
    Lifetime Mode=::NewEnumerator1
    Mass Max=1.5
    Mass Mode=::NewEnumerator3
    Position Offset=&RandomRangeVector(Minimum=<0 -10 -10>, Maximum=<0 10 10>)
    Sprite Rotation Mode=::NewEnumerator2
    Sprite Size Mode=::NewEnumerator1
    Uniform Sprite Size Max=4.5
    Uniform Sprite Size Min=3
    UsePositionOffset=true
    Write Material Random=true
   )
   &SampleParticlesFromOtherEmitter()
   &InitialMeshOrientation(Use Rotation Vector=false)
   &AddVelocity(Velocity=&Lerp_Vector(
     A=&RandomRangeVector(Minimum=<-50 -12 -5>, Maximum=<-113 -12 33>)
     B=&RandomRangeVector(Minimum=<-74 -12 -5>, Maximum=<-170 -12 33>)
     Alpha=&FloatFromCurve(FloatCurve=DataInterface() {propertyValues {Curve {keys [{tangentMode=RCTM_User, time=0, value=0}, {tangentMode=RCTM_User, time=0.7, value=1}]}}})
    ))
  ]
  particleUpdateScript [
   &ParticleState()
   &ScaleVelocity(Velocity Scale=&RandomRangeVector(Minimum=<0.5 0.9 0.9>, Maximum=<0.6 1.1 1.1>))
   &AerodynamicDrag(
    Aerodynamic Drag:NiagaraFloat=&FloatFromCurve(FloatCurve=DataInterface() {propertyValues {Curve {keys [{interpMode=RCIM_Cubic, tangentMode=RCTM_User, time=0, value=0.2}, {interpMode=RCIM_Cubic, tangentMode=RCTM_User, time=1, value=1}]}}})
    Aerodynamic Rotational Drag=0.15
    Pivot Offset Maximum:Vector3f=<5 0.5 1>
    Rotation Strength:NiagaraFloat=&RandomRangeFloat(Minimum=0.2, Maximum=0.3)
    Use Linear Drag=true
    Use Rotational Drag=false
   )
   &CurlNoiseForce(Bias Noise=false, Noise Frequency=59.917, Noise Strength=500, Pan Noise=false, Pan Noise Field=<976.2582 0 500>)
   &Drag(Drag=0.4537, Rotational Drag=0.2513, Use Linear Drag=true, Use Rotational Drag=true)
   &SolveForcesAndVelocity(Clamp Velocity=false, Limit Acceleration=false)
   &ScaleSpriteSize(Uniform Curve Sprite Scale=DataInterface() {propertyValues {Curve {keys [
        {arriveTangent=14.5835, interpMode=RCIM_Cubic, leaveTangent=14.5835, tangentMode=RCTM_User, time=0, value=0}
        {arriveTangent=0.9132, interpMode=RCIM_Cubic, leaveTangent=0.9132, tangentMode=RCTM_User, time=0.1, value=1}
        {arriveTangent=-1.1111, interpMode=RCIM_Cubic, leaveTangent=-1.1111, time=0.8786, value=1}
        {arriveTangent=-3.3546, interpMode=RCIM_Cubic, leaveTangent=-3.3546, tangentMode=RCTM_User, time=1, value=0}
       ]}}})
   &AlignSpriteToMeshOrientation(SetAlignment=true, SetFacing=true)
  ]
  rendererClasses [&NiagaraSpriteRendererProperties]
  renderers [{rendererClass=&NiagaraSpriteRendererProperties}]
 }
 Low_Petal_Tracer {
  emitterUpdateScript [
   &EmitterState(Life Cycle Mode=::NewEnumerator1, Loop Behavior=::NewEnumerator1, Loop Duration=1, UseLoopDelay=false)
   &SpawnBurst_Instantaneous(Spawn Count=2, Use Spawn Probability=false, UseLoopCountLimit=false)
  ]
  particleSpawnScript [&InitializeParticle(
    Color=<rgba 15 0 11.5 1>
    Lifetime=1.2
    Sprite Rotation Mode=::NewEnumerator2
    Sprite Size Mode=::NewEnumerator1
    Uniform Sprite Size Max=4
    Uniform Sprite Size Min=3
    UsePositionOffset=false
    Write Material Random=true
   )]
  particleUpdateScript [
   &ShapeLocation(
    MaskSpawnGroup=false
    Offset=&AddVectorToPosition(
     Position=&VectorFromCurve(
      VectorCurve=DataInterface() {propertyValues {
        XCurve {keys [<at 0 3>, <at 0.1 12>, {arriveTangent=16.9116, time=0.9, value=105}]}
        YCurve {keys [<at 0 0>,<at 0 0>]}
        ZCurve {keys [<at 0 0>,<at 1 0>]}
       }}
      Scale Curve=<-1 1 1>
     )
     Vector=<1 0 0>
    )
    Offset Mode=::NewEnumerator0
    Override Seed=false
    Sphere Radius=0
    UseOwnerScale=true
   )
   &ParticleState()
   /SetVariables(Particles.Position=&VectorFromCurve(
     VectorCurve=DataInterface() {propertyValues {XCurve {keys [<at 0.1 12>,<at 0.9 105>]}, YCurve {keys [<at 0 0>,<at 1 0>]}, ZCurve {keys [<at 0 0>,<at 1 0>]}}}
     Scale Curve=<-1 1 1>
    )) {enabled=false, isSetParameters=true}
  ]
  rendererClasses []
  renderers []
 }
 Low_Petals_Tight_To_Blade {
  emitterUpdateScript [
   &EmitterState(Life Cycle Mode=::NewEnumerator1, Loop Behavior=::NewEnumerator1, Loop Duration=1.9, UseLoopDelay=false)
   &SpawnParticlesFromOtherEmitter(
    Attribute Reader=DataInterface() {propertyValues {EmitterBinding {bindingMode=Other, emitterName=Low_Petal_Tracer}}}
    Spawn Rate=50
    Use Maximum Age=false
    Use Minimum Age=false
   )
  ]
  particleSpawnScript [
   &InitializeParticle(
    Color=<rgba 3 0.0156 1.5514 1>
    Lifetime Max=1.5
    Lifetime Min=0.3
    Lifetime Mode=::NewEnumerator1
    Mass Max=1.5
    Mass Mode=::NewEnumerator3
    Position Offset=&RandomRangeVector(Minimum=<0 -50 -50>, Maximum=<0 50 50>)
    Sprite Rotation Mode=::NewEnumerator2
    Sprite Size Mode=::NewEnumerator1
    Uniform Sprite Size Max=2
    Uniform Sprite Size Min=1
    UsePositionOffset=true
    Write Material Random=true
   )
   &SampleParticlesFromOtherEmitter()
   &InitialMeshOrientation(Use Rotation Vector=false)
   &AddVelocity(Velocity=&VectorFromCurve(
     VectorCurve=DataInterface() {propertyValues {XCurve {keys [<at 0 1>,<at 1 2>]}, YCurve {keys [<at 0 0.52>,<at 1 1.5>]}, ZCurve {keys [<at 0 0.52>,<at 1 1.5>]}}}
     Scale Curve=&RandomRangeVector(Minimum=<-10 -25 -25>, Maximum=<-10 25 25>)
    ))
  ]
  particleUpdateScript [
   &ParticleState()
   &ScaleVelocity(Velocity Scale=&RandomRangeVector(Minimum=<0.5 1.2 1.2>, Maximum=<0.6 1.25 1.25>))
   &AerodynamicDrag(
    Aerodynamic Drag:NiagaraFloat=&RandomRangeFloat(Minimum=0.3333, Maximum=1)
    Aerodynamic Rotational Drag=0.15
    Rotation Strength:NiagaraFloat=&RandomRangeFloat(Minimum=0.2, Maximum=0.3)
    Use Linear Drag=true
    Use Rotational Drag=false
   )
   &CurlNoiseForce(Bias Noise=false, Noise Frequency=5, Noise Strength=20, Pan Noise=false, Pan Noise Field=<976.2582 0 500>)
   &SolveForcesAndVelocity(Clamp Velocity=false, Limit Acceleration=false)
   &ScaleSpriteSize(Uniform Curve Sprite Scale=DataInterface() {propertyValues {Curve {keys [
        {arriveTangent=14.5835, interpMode=RCIM_Cubic, leaveTangent=14.5835, tangentMode=RCTM_User, time=0, value=0}
        {arriveTangent=1.1111, interpMode=RCIM_Cubic, leaveTangent=1.1111, time=0.1, value=1}
        {arriveTangent=-1.1111, interpMode=RCIM_Cubic, leaveTangent=-1.1111, time=0.9, value=1}
        {arriveTangent=-8.7113, interpMode=RCIM_Cubic, leaveTangent=-8.7113, tangentMode=RCTM_User, time=1, value=0}
       ]}}})
   &ScaleSpriteSizeBySpeed(Max Scale Factor=<3 3>)
   &AlignSpriteToMeshOrientation(SetAlignment=true, SetFacing=true)
   &ScaleColor(
    Scale Alpha=&FloatFromCurve(FloatCurve=DataInterface() {propertyValues {Curve {keys [{interpMode=RCIM_Cubic, tangentMode=RCTM_User, time=0, value=1}, {interpMode=RCIM_Cubic, tangentMode=RCTM_User, time=1, value=0}]}}})
    Scale RGB=&MakeVectorFromLinearColorRGB(Color=&ColorFromCurve(ColorCurve=DataInterface() {propertyValues {
        AlphaCurve {keys [<at 0 1 RCIM_Cubic>,<at 1 0 RCIM_Cubic>]}
        BlueCurve {keys [<at 0 50 RCIM_Cubic>,<at 1 1 RCIM_Cubic>]}
        GreenCurve {keys [<at 0 0 RCIM_Cubic>,<at 1 1 RCIM_Cubic>]}
        RedCurve {keys [<at 0 48.3334 RCIM_Cubic>,<at 1 1 RCIM_Cubic>]}
       }}))
    ScaleA=true
    ScaleRGB=true
   )
  ]
  rendererClasses [&NiagaraSpriteRendererProperties]
  renderers [{rendererClass=&NiagaraSpriteRendererProperties}]
 }
 Petal_Tracer {
  simTarget=GPUComputeSim
  emitterUpdateScript [
   &EmitterState(Life Cycle Mode=::NewEnumerator1, Loop Behavior=::NewEnumerator1, Loop Duration=1, UseLoopDelay=false)
   &SpawnBurst_Instantaneous(Spawn Count=2, Use Spawn Probability=false, UseLoopCountLimit=false)
  ]
  particleSpawnScript [&InitializeParticle(
    Color=<rgba 15 0 11.5 1>
    Lifetime=1.2
    Sprite Rotation Mode=::NewEnumerator2
    Sprite Size Mode=::NewEnumerator1
    Uniform Sprite Size Max=4
    Uniform Sprite Size Min=3
    UsePositionOffset=false
    Write Material Random=true
   )]
  particleUpdateScript [
   &ShapeLocation(
    MaskSpawnGroup=false
    Offset=&AddVectorToPosition(
     Position=&VectorFromCurve(
      VectorCurve=DataInterface() {propertyValues {
        XCurve {keys [<at 0 3>, <at 0.1 12>, {arriveTangent=16.9116, time=0.9, value=105}]}
        YCurve {keys [<at 0 0>,<at 0 0>]}
        ZCurve {keys [<at 0 0>,<at 1 0>]}
       }}
      Scale Curve=<-1 1 1>
     )
     Vector=<1 0 0>
    )
    Offset Mode=::NewEnumerator0
    Override Seed=false
    Sphere Radius=0
    UseOwnerScale=true
   )
   &ParticleState()
  ]
  rendererClasses []
  renderers []
 }
 Petals_Flying {
  simTarget=GPUComputeSim
  emitterUpdateScript [
   &EmitterState(Life Cycle Mode=::NewEnumerator1, Loop Behavior=::NewEnumerator1, Loop Duration=1.9, UseLoopDelay=false)
   &SpawnParticlesFromOtherEmitter(
    Attribute Reader=DataInterface() {propertyValues {EmitterBinding {bindingMode=Other, emitterName=Petal_Tracer}}}
    Spawn Rate=100
    Spawn Rate Per Particle Cap=200
    Use Maximum Age=false
    Use Minimum Age=false
   )
  ]
  particleSpawnScript [
   &InitializeParticle(
    Lifetime Max=1.6731
    Lifetime Min=0.666
    Lifetime Mode=::NewEnumerator1
    Mass Max=1.5
    Mass Mode=::NewEnumerator3
    Sprite Rotation Mode=::NewEnumerator2
    Sprite Size Mode=::NewEnumerator1
    Uniform Sprite Size Max=4.5
    Uniform Sprite Size Min=3
    UsePositionOffset=false
    Write Material Random=true
   )
   &SampleParticlesFromOtherEmitter()
   &ShapeLocation(MaskSpawnGroup=false, Offset Mode=::NewEnumerator0, Override Seed=false, Sphere Radius=0, UseOwnerScale=true)
   &InitialMeshOrientation(Use Rotation Vector=false)
   &AddVelocity(Velocity=&Lerp_Vector(
     A=&RandomRangeVector(Minimum=<-49.3333 -11.6667 -5>, Maximum=<-113.3333 11.6667 33.3333>)
     B=&RandomRangeVector(Minimum=<-74 -11.6667 -5>, Maximum=<-170 11.6667 33.3333>)
     Alpha=&FloatFromCurve(FloatCurve=DataInterface() {propertyValues {Curve {keys [{tangentMode=RCTM_User, time=0, value=0}, {tangentMode=RCTM_User, time=0.7, value=1}]}}})
    ))
  ]
  particleUpdateScript [
   &ParticleState()
   &WindForce(
    Offset By Time:Vector3f=<0.5 0.25 0.2>
    Scale=0.673
    Wind Speed=&RandomRangeVector(Minimum=<-50 -15 -5>, Maximum=<-75 15 40>)
    Wind Speed Scale=0.5
   )
   &AerodynamicDrag(
    Aerodynamic Drag:NiagaraFloat=&FloatFromCurve(FloatCurve=DataInterface() {propertyValues {Curve {keys [{interpMode=RCIM_Cubic, tangentMode=RCTM_User, time=0, value=0.2}, {interpMode=RCIM_Cubic, tangentMode=RCTM_User, time=1, value=1}]}}})
    Aerodynamic Rotational Drag=0.15
    Pivot Offset Maximum:Vector3f=<5 0.5 1>
    Rotation Strength:NiagaraFloat=&RandomRangeFloat(Minimum=0.2, Maximum=0.3)
    Use Linear Drag=true
    Use Rotational Drag=false
   )
   &CurlNoiseForce(Bias Noise=false, Noise Frequency=59.917, Noise Strength=500, Pan Noise=false, Pan Noise Field=<976.2582 0 500>)
   &Drag(Drag=0.4537, Rotational Drag=0.2513, Use Linear Drag=true, Use Rotational Drag=true)
   &SolveForcesAndVelocity(Clamp Velocity=false, Limit Acceleration=false)
   &ScaleSpriteSize(Uniform Curve Sprite Scale=DataInterface() {propertyValues {Curve {keys [
        {arriveTangent=14.5835, interpMode=RCIM_Cubic, leaveTangent=14.5835, tangentMode=RCTM_User, time=0, value=0}
        {arriveTangent=0.9132, interpMode=RCIM_Cubic, leaveTangent=0.9132, tangentMode=RCTM_User, time=0.1, value=1}
        {arriveTangent=-1.1111, interpMode=RCIM_Cubic, leaveTangent=-1.1111, time=0.8786, value=1}
        {arriveTangent=-3.3546, interpMode=RCIM_Cubic, leaveTangent=-3.3546, tangentMode=RCTM_User, time=1, value=0}
       ]}}})
   &AlignSpriteToMeshOrientation(SetAlignment=true, SetFacing=true)
  ]
  rendererClasses [&NiagaraSpriteRendererProperties]
  renderers [{rendererClass=&NiagaraSpriteRendererProperties}]
 }
 Petals_Flying001 {
  simTarget=GPUComputeSim
  emitterUpdateScript [
   &EmitterState(Life Cycle Mode=::NewEnumerator1, Loop Behavior=::NewEnumerator1, Loop Duration=1.9, UseLoopDelay=false)
   &SpawnParticlesFromOtherEmitter(
    Attribute Reader=DataInterface() {propertyValues {EmitterBinding {bindingMode=Other, emitterName=Petal_Tracer}}}
    Spawn Rate=100
    Spawn Rate Per Particle Cap=300
    Use Maximum Age=false
    Use Minimum Age=false
   )
  ]
  particleSpawnScript [
   &InitializeParticle(
    Color=<rgba 3 0.0156 1.5514 1>
    Lifetime Max=1.5
    Lifetime Min=0.3
    Lifetime Mode=::NewEnumerator1
    Mass Max=1.5
    Mass Mode=::NewEnumerator3
    Sprite Rotation Mode=::NewEnumerator2
    Sprite Size Mode=::NewEnumerator1
    Uniform Sprite Size Max=2
    Uniform Sprite Size Min=1
    UsePositionOffset=false
    Write Material Random=true
   )
   &SampleParticlesFromOtherEmitter()
   &ShapeLocation(MaskSpawnGroup=false, Offset Mode=::NewEnumerator0, Override Seed=false, Sphere Radius=0, UseOwnerScale=true)
   &InitialMeshOrientation(Use Rotation Vector=false)
   &AddVelocity(Velocity=&VectorFromCurve(
     VectorCurve=DataInterface() {propertyValues {XCurve {keys [<at 0 1>,<at 1 2>]}, YCurve {keys [<at 0 0>,<at 1 0>]}, ZCurve {keys [<at 0 0>,<at 1 0>]}}}
     Scale Curve=<-10 1 1>
    ))
  ]
  particleUpdateScript [
   &ParticleState()
   &WindForce(
    Offset By Time:Vector3f=<0.5 0.25 0.2>
    Scale=0.673
    Wind Speed=&RandomRangeVector(Minimum=<-50 -15 -5>, Maximum=<-75 15 40>)
    Wind Speed Scale=0.5
   )
   &AerodynamicDrag(
    Aerodynamic Drag:NiagaraFloat=&RandomRangeFloat(Minimum=0.3333, Maximum=1)
    Aerodynamic Rotational Drag=0.15
    Rotation Strength:NiagaraFloat=&RandomRangeFloat(Minimum=0.2, Maximum=0.3)
    Use Linear Drag=true
    Use Rotational Drag=false
   )
   &CurlNoiseForce(Bias Noise=false, Noise Frequency=5, Noise Strength=20, Pan Noise=false, Pan Noise Field=<976.2582 0 500>)
   &SolveForcesAndVelocity(Clamp Velocity=false, Limit Acceleration=false)
   &ScaleSpriteSize(Uniform Curve Sprite Scale=DataInterface() {propertyValues {Curve {keys [
        {arriveTangent=14.5835, interpMode=RCIM_Cubic, leaveTangent=14.5835, tangentMode=RCTM_User, time=0, value=0}
        {arriveTangent=1.1111, interpMode=RCIM_Cubic, leaveTangent=1.1111, time=0.1, value=1}
        {arriveTangent=-1.1111, interpMode=RCIM_Cubic, leaveTangent=-1.1111, time=0.9, value=1}
        {arriveTangent=-8.7113, interpMode=RCIM_Cubic, leaveTangent=-8.7113, tangentMode=RCTM_User, time=1, value=0}
       ]}}})
   &ScaleSpriteSizeBySpeed(Max Scale Factor=<3 3>)
   &AlignSpriteToMeshOrientation(SetAlignment=true, SetFacing=true)
   &ScaleColor(
    Scale Alpha=&FloatFromCurve(FloatCurve=DataInterface() {propertyValues {Curve {keys [{interpMode=RCIM_Cubic, tangentMode=RCTM_User, time=0, value=1}, {interpMode=RCIM_Cubic, tangentMode=RCTM_User, time=1, value=0}]}}})
    Scale RGB=&MakeVectorFromLinearColorRGB(Color=&ColorFromCurve(ColorCurve=DataInterface() {propertyValues {
        AlphaCurve {keys [<at 0 1 RCIM_Cubic>,<at 1 0 RCIM_Cubic>]}
        BlueCurve {keys [<at 0 50 RCIM_Cubic>,<at 1 1 RCIM_Cubic>]}
        GreenCurve {keys [<at 0 0 RCIM_Cubic>,<at 1 1 RCIM_Cubic>]}
        RedCurve {keys [<at 0 48.3334 RCIM_Cubic>,<at 1 1 RCIM_Cubic>]}
       }}))
    ScaleA=true
    ScaleRGB=true
   )
  ]
  rendererClasses [&NiagaraSpriteRendererProperties]
  renderers [{rendererClass=&NiagaraSpriteRendererProperties}]
 }
}
summary {
 systemName=NS_Emote_ZenMaster_Petals
 emitters [
  {emitterName=Petals_Flying, simTarget=GPUComputeSim, rendererClasses [&NiagaraSpriteRendererProperties]}
  {emitterName=Petal_Tracer, simTarget=GPUComputeSim, rendererClasses []}
  {emitterName=Petals_Flying001, simTarget=GPUComputeSim, rendererClasses [&NiagaraSpriteRendererProperties]}
  {
   emitterName=Directional_Flare
   rendererClasses [&NiagaraSpriteRendererProperties, &NiagaraLightRendererProperties]
  }
  {emitterName=Low_Petals_Tight_To_Blade, rendererClasses [&NiagaraSpriteRendererProperties]}
  {emitterName=Low_Petal_Plume, rendererClasses [&NiagaraSpriteRendererProperties]}
  {emitterName=Low_Petal_Tracer, rendererClasses []}
 ]
 userVariables [
  User.bIsFrontEnd {type=NiagaraBool}
  User.bIsFrontEndPreview {type=NiagaraBool}
  User.ParaGlideLeanAlpha {type=NiagaraFloat, defaultValue=0, description=""}
 ]
}
